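# Niagara idle effect for the TaperSlinkAcre cosmetic shoes, emitting sound-wave sprites from both feet.
# Particles are partitioned per shoe, follow skeletal mesh sockets, and are sized by clamped left and right shoe scale.
# Gameplay spawn probability follows a normalized-velocity curve, while off-pawn and primary front-end previews gate it on a repeating ten-second timer.
# Each sprite grows over its short life, flashing in and fading while shifting color through curves.
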NiagaraSystem /CosmeticShoes/Assets/FORT_Shoes/Shoe_TaperSlinkAcre/FX/NS_Shoes_TaperSlinkAcre_Idle
dependencies {
 usedDataInterfaces [&NiagaraDataInterfaceCurve, &NiagaraDataInterfaceColorCurve, &NiagaraDataInterfaceSkeletalMesh]
 usedDynamicInputs [
  &Multiply_Vector
  &RemapRange
  &VectorLength
 ]
 usedModules [
  &self:SystemSpawnScript.SystemScriptSource.SystemScriptGraph.NiagaraNodeAssignment_2.SetVariables_#1
  &SystemState
  &InterpolateOverTime
  &self:SystemSpawnScript.SystemScriptSource.SystemScriptGraph.NiagaraNodeAssignment_3.SetVariables_#2
  &self:SystemSpawnScript.SystemScriptSource.SystemScriptGraph.NiagaraNodeAssignment_4.SetVariables_#3
 ]
}
emitters {NE_SoundWave {
  emitterUpdateScript [
   &EmitterState()
   &SpawnRate(
    Spawn Probability=&FN_FrontEndFloatModifier(
     Gameplay=&MakeCustomFloatFromBool(
      Bool=$User.IsOffPawnPreview
      True Float=&MakeCustomFloatFromBool(
       Bool=&SetBoolByFloatComparison(A=&ModuloFloat(A=$Engine.System.Age, B=10), B=0.7, Comparison Type=::NewEnumerator6)
       True Float=1
       False Float=0
      )
      False Float=&FloatFromCurve(
       FloatCurve=DataInterface() {propertyValues {Curve {
          defaultValue=1
          keys [
           {arriveTangent=17.7955, interpMode=RCIM_Cubic, leaveTangent=17.7955, tangentMode=RCTM_User, time=0.3, value=0}
           {arriveTangent=-0.201, interpMode=RCIM_Cubic, leaveTangent=-0.201, tangentMode=RCTM_User, time=0.4578, value=1}
           {arriveTangent=-0.0595, interpMode=RCIM_Cubic, leaveTangent=-0.0595, tangentMode=RCTM_User, time=1, value=1}
          ]
         }}}
       CurveIndex=$System.NormalizedVelocityRange
      )
     )
     Front End=&MakeCustomFloatFromBool(
      Bool=&Boolean_AndOperation(
       Bool1=$User.bIsPrimaryPreviewItem
       Bool2=&SetBoolByFloatComparison(A=&ModuloFloat(A=$Engine.System.Age, B=10), B=0.7, Comparison Type=::NewEnumerator6)
      )
      True Float=1
      False Float=0
     )
    )
    SpawnRate=10
    Use Spawn Probability=true
   )
  ]
  particleSpawnScript [
   &PartitionParticles()
   &InitializeParticle(
    Color=<rgba 2 2 2 1>
    Lifetime=0.75
    Sprite Rotation Mode=::NewEnumerator2
    Sprite Size Mode=::NewEnumerator3
    Uniform Sprite Size=&MergeHelper_FloatSwitch(
     0=&Multiply_Float(
      A=35
      B=&ClampFloat(Float=&MakeFloatFromVector(Vector=$User.ShoeScale_L, Channel=::NewEnumerator0), Max=1.5)
     )
     1=&Multiply_Float(
      A=35
      B=&ClampFloat(Float=&MakeFloatFromVector(Vector=$User.ShoeScale_R, Channel=::NewEnumerator0), Max=1.5)
     )
     2=?
     3=?
     Index=$Particles.PartitionParticles.Partition
     Num=::NewEnumerator0
    )
    UsePositionOffset=false
    Write Material Random=true
   )
   &CameraOffset(Camera Offset Amount=10)
  ]
  particleUpdateScript [
   &ParticleState()
   &SocketLocation(
    SamplePosOffset=false
    Socket Index=$Particles.PartitionParticles.Partition
    Socket Reader=$System.Socket Reader
    Socket Sampling Mode=::NewEnumerator2
    VelocitySpeedLimit=false
   )
   &SkeletalMeshLocation(
    Bone / Socket Index=$Particles.PartitionParticles.Partition
    MaskSpawnGroup=false
    Mesh Sampling Type=::NewEnumerator5
    Skeletal Mesh=$System.Skeletal Mesh Reader
    Socket Sampling Mode=::NewEnumerator5
   )
   &ScaleSpriteSize(Uniform Curve Sprite Scale=DataInterface() {propertyValues {Curve {keys [
        {arriveTangent=1.2796, interpMode=RCIM_Cubic, leaveTangent=1.2796, tangentMode=RCTM_User, time=0, value=0}
        {arriveTangent=0.1491, interpMode=RCIM_Cubic, leaveTangent=0.1491, tangentMode=RCTM_User, time=1, value=1}
       ]}}})
   &ScaleColor(
    Scale Alpha=&FloatFromCurve(
     FloatCurve=DataInterface() {propertyValues {Curve {
        defaultValue=0
        keys [
         {interpMode=RCIM_Cubic, tangentMode=RCTM_User, time=0, value=0}
         {arriveTangent=8.798, interpMode=RCIM_Cubic, leaveTangent=8.798, tangentMode=RCTM_User, time=0.0795, value=0.6811}
         {arriveTangent=-3.7651, interpMode=RCIM_Cubic, leaveTangent=-3.7651, tangentMode=RCTM_User, time=0.1408, value=1}
         {arriveTangent=-1.6384, interpMode=RCIM_Cubic, leaveTangent=-1.6384, tangentMode=RCTM_User, time=0.25, value=0.5}
         {arriveTangent=0.0704, tangentMode=RCTM_User, time=1, value=0}
        ]
       }}}
     Scale Curve=1.5
    )
    Scale RGB=&MakeVectorFromLinearColorRGB(Color=&ColorFromCurve(ColorCurve=DataInterface() {propertyValues {
        AlphaCurve {keys [<at 0 1 RCIM_Cubic>]}
        BlueCurve {keys [<at 0.3755 0.0144 RCIM_Cubic>,<at 0.8407 0.0594 RCIM_Cubic>,<at 1 0.4365>]}
        GreenCurve {keys [<at 0.3755 0.8 RCIM_Cubic>,<at 0.8407 0.5312 RCIM_Cubic>,<at 1 0.5312>]}
        RedCurve {keys [<at 0.3755 0.5428 RCIM_Cubic>,<at 0.8407 0.0221 RCIM_Cubic>,<at 1 0>]}
       }}))
    ScaleA=true
    ScaleRGB=true
   )
  ]
  rendererClasses [&NiagaraSpriteRendererProperties]
  renderers [{rendererClass=&NiagaraSpriteRendererProperties}]
 }}
summary {
 systemName=NS_Shoes_TaperSlinkAcre_Idle
 emitters [{emitterName=NE_SoundWave, rendererClasses [&NiagaraSpriteRendererProperties]}]
 userVariables [
  User.bIsFrontEnd {type=NiagaraBool}
  User.bIsFrontEndPreview {type=NiagaraBool}
  User.bIsPrimaryPreviewItem {type=NiagaraBool}
  User.IsOffPawnPreview {type=NiagaraBool}
  User.ShoeScale_L {type=Vector3f, description="", defaultValue {struct=Vector3f, value=<1 1 1>}}
  User.ShoeScale_R {type=Vector3f, description="", defaultValue {struct=Vector3f, value=<1 1 1>}}
 ]
}
