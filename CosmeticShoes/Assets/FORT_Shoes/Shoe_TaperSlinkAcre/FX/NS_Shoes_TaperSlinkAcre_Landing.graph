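# Niagara landing effect for the TaperSlinkAcre cosmetic shoes, firing one-shot bursts of green-tinted sprite sound waves, sparks and spikes at both feet.
# PartitionParticles splits each burst between socket-reader sockets, with per-foot User.ShoeScale_L/R and System.ScaleShoes driving size.
# NE_SoundWave follows its sockets each update with staggered growth.
# The other emitters spawn at offset sockets: up-facing waves, curl-noise sparks and rising spikes, the latter two in rings.
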NiagaraSystem /CosmeticShoes/Assets/FORT_Shoes/Shoe_TaperSlinkAcre/FX/NS_Shoes_TaperSlinkAcre_Landing
dependencies {
 usedDataInterfaces [&NiagaraDataInterfaceCurve, &NiagaraDataInterfaceColorCurve, &NiagaraDataInterfaceSocketReader]
 usedDynamicInputs [&RandomRangeVector, &ClampFloat]
 usedModules [
  &self:SystemSpawnScript.SystemScriptSource.SystemScriptGraph.NiagaraNodeAssignment_0.SetVariables_#1
  &self:SystemSpawnScript.SystemScriptSource.SystemScriptGraph.NiagaraNodeAssignment_1.SetVariables_#2
  &SystemState
 ]
}
emitters {
 NE_SoundWave {
  emitterUpdateScript [
   &EmitterState()
   &SpawnBurst_Instantaneous(Spawn Count=6, Use Spawn Probability=false, UseLoopCountLimit=false)
  ]
  particleSpawnScript [
   &PartitionParticles(Normalized Index In Partition Includes One=true, Particles Per Partition=3, Partition Mode:ENiagara_PartitionMode_v2=::NewEnumerator1)
   &InitializeParticle(
    Color=<rgba 2 2 2 1>
    Lifetime=0.5
    Sprite Rotation Mode=::NewEnumerator2
    Sprite Size Mode=::NewEnumerator3
    Uniform Sprite Size=&FloatFromCurve(
     FloatCurve=DataInterface() {propertyValues {Curve {defaultValue=0.25, keys [{tangentMode=RCTM_User, time=0, value=0.5}, {tangentMode=RCTM_User, time=1, value=1}]}}}
     CurveIndex=$Particles.PartitionParticles.Normalized Index In Partition
     Scale Curve=&MaskFloatBySpawnGroup(
      Masked Float=&Multiply_Float(A=35, B=&MakeFloatFromVector(Vector=$User.ShoeScale_L, Channel=::NewEnumerator0))
      Passthrough Float=&Multiply_Float(A=35, B=&MakeFloatFromVector(Vector=$User.ShoeScale_R, Channel=::NewEnumerator0))
      Spawn Group Mask=$Particles.PartitionParticles.Partition
     )
    )
    UsePositionOffset=false
    Write Material Random=true
   )
   &CameraOffset(Camera Offset Amount=10)
  ]
  particleUpdateScript [
   &ParticleState()
   &SocketLocation(
    SamplePosOffset=false
    Socket Index=$Particles.PartitionParticles.Partition
    Socket Reader=$System.Socket Reader
    Socket Sampling Mode=::NewEnumerator3
    VelocitySpeedLimit=false
   )
   &ScaleSpriteSize(
    Uniform Curve Index=&NormalizeFloat(
     Value=&Multiply_Float(
      A=$Particles.NormalizedAge
      B=&FloatFromCurve(
       FloatCurve=DataInterface() {propertyValues {Curve {defaultValue=0.5, keys [{tangentMode=RCTM_User, time=0, value=0.5}, {tangentMode=RCTM_User, time=1, value=1}]}}}
       CurveIndex=$Particles.PartitionParticles.Normalized Index In Partition
      )
     )
     Min=0
     Max=1
    )
    Uniform Curve Scale=$System.ScaleShoes
    Uniform Curve Sprite Scale=DataInterface() {propertyValues {Curve {keys [
        {arriveTangent=1.2796, interpMode=RCIM_Cubic, leaveTangent=1.2796, tangentMode=RCTM_User, time=0, value=0}
        {arriveTangent=0.1491, interpMode=RCIM_Cubic, leaveTangent=0.1491, tangentMode=RCTM_User, time=1, value=1}
       ]}}}
   )
   &ScaleColor(
    Scale Alpha=&Multiply_Float(
     A=&FloatFromCurve(
      FloatCurve=DataInterface() {propertyValues {Curve {
         defaultValue=1
         keys [
          {arriveTangent=-0.1097, interpMode=RCIM_Cubic, leaveTangent=-0.1097, tangentMode=RCTM_User, time=0.2, value=1}
          {interpMode=RCIM_Cubic, tangentMode=RCTM_User, time=0.5, value=0}
         ]
        }}}
      CurveIndex=$Engine.System.Age
     )
     B=&FloatFromCurve(
      FloatCurve=DataInterface() {propertyValues {Curve {
         defaultValue=0
         keys [
          {interpMode=RCIM_Cubic, tangentMode=RCTM_User, time=0, value=0}
          {interpMode=RCIM_Cubic, tangentMode=RCTM_User, time=0.5, value=1}
          {tangentMode=RCTM_User, time=1, value=0}
         ]
        }}}
      CurveIndex=&NormalizeFloat(
       Value=&Multiply_Float(
        A=$Particles.NormalizedAge
        B=&FloatFromCurve(
         FloatCurve=DataInterface() {propertyValues {Curve {defaultValue=0.5, keys [{tangentMode=RCTM_User, time=0, value=0.5}, {tangentMode=RCTM_User, time=1, value=1.5}]}}}
         CurveIndex=$Particles.PartitionParticles.Normalized Index In Partition
        )
       )
       Min=0
       Max=1
      )
      Scale Curve=1.5
     )
    )
    Scale RGB=&MakeVectorFromLinearColorRGB(Color=&ColorFromCurve(ColorCurve=DataInterface() {propertyValues {
        AlphaCurve {keys [<at 0 1 RCIM_Cubic>]}
        BlueCurve {keys [<at 0.3755 0.0144 RCIM_Cubic>,<at 0.8407 0.0594 RCIM_Cubic>,<at 1 0.4365>]}
        GreenCurve {keys [<at 0.3755 0.8 RCIM_Cubic>,<at 0.8407 0.5312 RCIM_Cubic>,<at 1 0.5312>]}
        RedCurve {keys [<at 0.3755 0.5428 RCIM_Cubic>,<at 0.8407 0.0221 RCIM_Cubic>,<at 1 0>]}
       }}))
    ScaleA=true
    ScaleRGB=true
   )
  ]
  rendererClasses [&NiagaraSpriteRendererProperties]
  renderers [{rendererClass=&NiagaraSpriteRendererProperties}]
 }
 NE_SoundWave_Running {
  emitterUpdateScript [
   &EmitterState()
   &SpawnBurst_Instantaneous(Spawn Count=6, Use Spawn Probability=false, UseLoopCountLimit=false)
  ]
  particleSpawnScript [
   &PartitionParticles()
   &InitializeParticle(
    Lifetime=&FloatFromCurve(
     FloatCurve=DataInterface() {propertyValues {Curve {defaultValue=0.5, keys [{tangentMode=RCTM_User, time=0, value=1}, {tangentMode=RCTM_User, time=1, value=0.5}]}}}
     CurveIndex=&ReturnNormalizedExecIndex()
     Scale Curve=0.5
    )
    Sprite Rotation Mode=::NewEnumerator2
    Sprite Size Mode=::NewEnumerator3
    Uniform Sprite Size=&FloatFromCurve(
     FloatCurve=DataInterface() {propertyValues {Curve {defaultValue=1, keys [{tangentMode=RCTM_User, time=0, value=0.25}, {tangentMode=RCTM_User, time=1, value=1}]}}}
     CurveIndex=&ReturnNormalizedExecIndex()
     Scale Curve=&MaskFloatBySpawnGroup(
      Masked Float=&Multiply_Float(A=100, B=&MakeFloatFromVector(Vector=$User.ShoeScale_L, Channel=::NewEnumerator0))
      Passthrough Float=&Multiply_Float(A=100, B=&MakeFloatFromVector(Vector=$User.ShoeScale_R, Channel=::NewEnumerator0))
      Spawn Group Mask=$Particles.PartitionParticles.Partition
     )
    )
    UsePositionOffset=false
    Write Material Random=true
   )
   &SocketLocation(
    SamplePosOffset=true
    Sampled Position Offset=<0 0 -10>
    Sampled Position Offset Coordinate Space:ENiagaraSkelMeshTransforms=::NewEnumerator1
    Socket Index=&MergeHelper_IntSwitch(0=2, 1=3, 2=?, 3=?, 4=?, 5=?, Index=$Particles.PartitionParticles.Partition, Num=::NewEnumerator0)
    Socket Reader=$System.Socket Reader
    Socket Sampling Mode=::NewEnumerator3
    VelocitySpeedLimit=false
   )
   &CameraOffset(Camera Offset Amount=10)
   &SpriteFacingAndAlignment(Facing Coordinate Space=::Local, Sprite Facing=<0 0 1>, WriteAlignment=true, WriteFacing=true)
  ]
  particleUpdateScript [
   &ParticleState()
   &ScaleSpriteSize(
    Uniform Curve Scale=$System.ScaleShoes
    Uniform Curve Sprite Scale=DataInterface() {propertyValues {Curve {keys [
        {arriveTangent=1.2796, interpMode=RCIM_Cubic, leaveTangent=1.2796, tangentMode=RCTM_User, time=0, value=0.25}
        {arriveTangent=0.1491, interpMode=RCIM_Cubic, leaveTangent=0.1491, tangentMode=RCTM_User, time=1, value=1}
       ]}}}
   )
   &ScaleColor(
    Scale Alpha=&FloatFromCurve(
     FloatCurve=DataInterface() {propertyValues {Curve {
        defaultValue=0
        keys [
         {interpMode=RCIM_Cubic, tangentMode=RCTM_User, time=0, value=0}
         {arriveTangent=8.798, interpMode=RCIM_Cubic, leaveTangent=8.798, tangentMode=RCTM_User, time=0.1404, value=0.7294}
         {arriveTangent=-3.7651, interpMode=RCIM_Cubic, leaveTangent=-3.7651, tangentMode=RCTM_User, time=0.2034, value=1}
         {arriveTangent=-1.2553, interpMode=RCIM_Cubic, leaveTangent=-1.2553, time=0.3197, value=0.3232}
         {tangentMode=RCTM_User, time=1, value=0}
        ]
       }}}
     Scale Curve=1.5
    )
    Scale RGB=&MakeVectorFromLinearColorRGB(Color=&ColorFromCurve(ColorCurve=DataInterface() {propertyValues {
        AlphaCurve {keys [<at 0 1 RCIM_Cubic>]}
        BlueCurve {keys [<at 0.3755 0.0144 RCIM_Cubic>,<at 0.8407 0.0594 RCIM_Cubic>,<at 1 0.4365>]}
        GreenCurve {keys [<at 0.3755 0.8 RCIM_Cubic>,<at 0.8407 0.5312 RCIM_Cubic>,<at 1 0.5312>]}
        RedCurve {keys [<at 0.3755 0.5428 RCIM_Cubic>,<at 0.8407 0.0221 RCIM_Cubic>,<at 1 0>]}
       }}))
    ScaleA=true
    ScaleRGB=true
   )
  ]
  rendererClasses [&NiagaraSpriteRendererProperties]
  renderers [{rendererClass=&NiagaraSpriteRendererProperties}]
 }
 On_Land_Spark {
  emitterUpdateScript [
   &EmitterState()
   &SpawnBurst_Instantaneous(Spawn Count=23, Use Spawn Probability=false, UseLoopCountLimit=false)
  ]
  particleSpawnScript [
   &InitializeParticle(
    Color=&ColorFromCurve(
     ColorCurve=DataInterface() {propertyValues {
       AlphaCurve {keys [<at 0 1 RCIM_Constant>]}
       BlueCurve {keys [<at 0 0 RCIM_Constant>,<at 0.7405 0.0306 RCIM_Constant>]}
       GreenCurve {keys [<at 0 3 RCIM_Constant>,<at 0.7405 1 RCIM_Constant>]}
       RedCurve {keys [<at 0 0.8348 RCIM_Constant>,<at 0.7405 0.4015 RCIM_Constant>]}
      }}
     CurveIndex=&RandomRangeFloat(Minimum=0, Maximum=1)
    )
    Lifetime=&RandomRangeFloat(Minimum=0.2, Maximum=0.4)
    Sprite Size Mode=::NewEnumerator1
    Uniform Sprite Size Max=2.5
    Uniform Sprite Size Min=1.2
    UsePositionOffset=false
    Write Material Random=true
   )
   &PartitionParticles()
   &SocketLocation(
    SamplePosOffset=true
    Sampled Position Offset=<0 0 -10>
    Sampled Position Offset Coordinate Space:ENiagaraSkelMeshTransforms=::NewEnumerator1
    Socket Index=&MergeHelper_IntSwitch(0=2, 1=3, 2=?, 3=?, 4=?, 5=?, Index=$Particles.PartitionParticles.Partition, Num=::NewEnumerator0)
    Socket Reader=$System.Socket Reader
    Socket Sampling Mode=::NewEnumerator3
    VelocitySpeedLimit=false
   )
   &ShapeLocation(
    Cylinder Distribution=::NewEnumerator2
    Cylinder Height=0.1
    Cylinder Radius=8
    MaskSpawnGroup=false
    Shape Primitive=::NewEnumerator1
    U Position=&ReturnNormalizedExecIndex()
    UseOwnerScale=true
   )
   &AddVelocity(Velocity=$Particles.ShapeLocation.ShapeNormal, Velocity Speed Scale=250) {module=AddVelocity001}
   &AddVelocity(Velocity=<0 0 70>)
  ]
  particleUpdateScript [
   &ParticleState()
   &CurlNoiseForce(Noise Frequency=60, Noise Quality / Cost=::NewEnumerator0, Noise Strength=1500, Pan Noise=false)
   &Drag(Drag=4, Use Linear Drag=true, Use Rotational Drag=true) {module=Drag001}
   &SolveForcesAndVelocity(Clamp Velocity=false, Limit Acceleration=false)
   &ScaleSpriteSizeBySpeed(Max Scale Factor=<2 7>, Min Scale Factor=<1 2>)
   &ScaleSpriteSize(
    Uniform Curve Scale=&MaskFloatBySpawnGroup(
     Masked Float=&Multiply_Float(A=$System.ScaleShoes, B=&MakeFloatFromVector(Vector=$User.ShoeScale_L, Channel=::NewEnumerator0))
     Passthrough Float=&Multiply_Float(A=$System.ScaleShoes, B=&MakeFloatFromVector(Vector=$User.ShoeScale_R, Channel=::NewEnumerator0))
     Spawn Group Mask=$Particles.PartitionParticles.Partition
    )
    Uniform Curve Sprite Scale=DataInterface() {propertyValues {Curve {keys [{interpMode=RCIM_Cubic, tangentMode=RCTM_User, time=0, value=1}, {interpMode=RCIM_Cubic, tangentMode=RCTM_User, time=1, value=0}]}}}
   )
   &ScaleColor(
    Linear Color Curve=DataInterface() {propertyValues {
      AlphaCurve {keys [<at 0 1>,<at 0.5828 1>,<at 1 0>]}
      BlueCurve {keys [<at 0.1576 1>,<at 0.426 1>,<at 0.6386 0.0469>]}
      GreenCurve {keys [<at 0.1576 1>,<at 0.426 1>,<at 0.6386 0.0469>]}
      RedCurve {keys [<at 0.1576 1>,<at 0.426 1>,<at 0.6386 0.0469>]}
     }}
    Scale Mode=::NewEnumerator2
   )
  ]
  rendererClasses [&NiagaraSpriteRendererProperties]
  renderers [{rendererClass=&NiagaraSpriteRendererProperties}]
 }
 Spikes {
  emitterUpdateScript [
   &EmitterState()
   &SpawnBurst_Instantaneous(Spawn Count=20, Use Spawn Probability=false, UseLoopCountLimit=false)
  ]
  particleSpawnScript [
   &InitializeParticle(
    Color=<rgba 0.487 1 0 0.2609>
    Lifetime Max=0.28
    Lifetime Min=0.25
    Lifetime Mode=::NewEnumerator1
    Sprite Rotation Mode=::NewEnumerator0
    Sprite Size Max=<15 60>
    Sprite Size Min=<10 20>
    Sprite Size Mode=::NewEnumerator2
    UsePositionOffset=false
    Write Material Random=true
   )
   &PartitionParticles()
   &SocketLocation(
    SamplePosOffset=true
    Sampled Position Offset=<0 0 -10>
    Sampled Position Offset Coordinate Space:ENiagaraSkelMeshTransforms=::NewEnumerator1
    Socket Index=&MergeHelper_IntSwitch(0=2, 1=3, 2=?, 3=?, 4=?, 5=?, Index=$Particles.PartitionParticles.Partition, Num=::NewEnumerator0)
    Socket Reader=$System.Socket Reader
    Socket Sampling Mode=::NewEnumerator3
    VelocitySpeedLimit=false
   )
   &ShapeLocation(
    Cylinder Distribution=::NewEnumerator2
    Cylinder Height=0
    Cylinder Height Midpoint=0
    Cylinder Radius=10
    MaskSpawnGroup=false
    Shape Primitive=::NewEnumerator1
    U Position=&ReturnNormalizedExecIndex()
    UseOwnerScale=true
   )
   &AddVelocity(Velocity=$Particles.ShapeLocation.ShapeNormal, Velocity Speed Scale=10) {module=AddVelocity001}
   &AddVelocity(Velocity=<0 0 1.25>, Velocity Speed Scale=&RandomRangeFloat(Minimum=3, Maximum=7))
  ]
  particleUpdateScript [
   &ParticleState()
   &SolveForcesAndVelocity(Clamp Velocity=false, Limit Acceleration=false)
   &ScaleSpriteSize(
    Uniform Curve Scale=&MaskFloatBySpawnGroup(
     Masked Float=&Multiply_Float(A=$System.ScaleShoes, B=&MakeFloatFromVector(Vector=$User.ShoeScale_L, Channel=::NewEnumerator0))
     Passthrough Float=&Multiply_Float(A=$System.ScaleShoes, B=&MakeFloatFromVector(Vector=$User.ShoeScale_R, Channel=::NewEnumerator0))
     Spawn Group Mask=$Particles.PartitionParticles.Partition
    )
    Uniform Curve Sprite Scale=DataInterface() {propertyValues {Curve {keys [
        {arriveTangent=1.6739, interpMode=RCIM_Cubic, leaveTangent=1.6739, tangentMode=RCTM_User, time=0, value=0}
        {arriveTangent=0.742, interpMode=RCIM_Cubic, leaveTangent=0.742, tangentMode=RCTM_User, time=1, value=1}
       ]}}}
   )
   &ScaleColor(
    Scale Alpha=&FloatFromCurve(FloatCurve=DataInterface() {propertyValues {Curve {keys [{interpMode=RCIM_Cubic, tangentMode=RCTM_User, time=0, value=1}, {interpMode=RCIM_Cubic, tangentMode=RCTM_User, time=1, value=0}]}}})
    Scale RGB=&Multiply_VectorByFloat(
     Vector=&VectorFromFloat(Value=&FloatFromCurve(FloatCurve=DataInterface() {propertyValues {Curve {keys [{interpMode=RCIM_Cubic, tangentMode=RCTM_User, time=0, value=1}, {interpMode=RCIM_Cubic, tangentMode=RCTM_User, time=1, value=0}]}}}))
     Float=1
    )
    ScaleA=true
    ScaleRGB=true
   )
  ]
  rendererClasses [&NiagaraSpriteRendererProperties]
  renderers [{rendererClass=&NiagaraSpriteRendererProperties}]
 }
}
summary {
 systemName=NS_Shoes_TaperSlinkAcre_Landing
 emitters [
  {emitterName=NE_SoundWave, rendererClasses [&NiagaraSpriteRendererProperties]}
  {emitterName=On_Land_Spark, rendererClasses [&NiagaraSpriteRendererProperties]}
  {emitterName=Spikes, rendererClasses [&NiagaraSpriteRendererProperties]}
  {emitterName=NE_SoundWave_Running, rendererClasses [&NiagaraSpriteRendererProperties]}
 ]
 userVariables [
  User.ShoeScale_L {type=Vector3f, description="", defaultValue {struct=Vector3f, value=<1 1 1>}}
  User.ShoeScale_R {type=Vector3f, description="", defaultValue {struct=Vector3f, value=<1 1 1>}}
 ]
}
